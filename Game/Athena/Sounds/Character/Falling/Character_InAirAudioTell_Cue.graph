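# Sound cue for the looping wind heard from a falling character in Athena, with an audio-tell variant for enemies.
# An Affiliation switch picks player foley or third-person enemy sound classes, each with occluded attenuation; the enemy branch adds a long HRTF capsule.
# Quality level selects layered mixes or a single cheaper loop per branch.
# Each loop fades in through an enveloper, and the FallDistance parameter scales its pitch and volume.
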SoundCue /Game/Athena/Sounds/Character/Falling/Character_InAirAudioTell_Cue
root=Switch_0
nodes [
 {
  id=Attenuation_0
  children [QualityLevel_0]
  properties {
   bOverrideAttenuation=true
   attenuationOverrides {
    bEnableOcclusion=true
    bEnableReverbSend=false
    falloffDistance=1500
    occlusionInterpolationTime=0.5
    occlusionLowPassFilterFrequency=6000
    occlusionTraceChannel=ECC_GameTraceChannel11
    occlusionVolumeAttenuation=3
    stereoSpread=768
    attenuationShapeExtents {x=0}
   }
  }
 }
 {
  id=Attenuation_1
  children [QualityLevel_2]
  properties {
   bOverrideAttenuation=true
   attenuationOverrides {
    attenuationShape=Capsule
    bEnableOcclusion=true
    bEnableReverbSend=false
    distanceAlgorithm=Custom
    falloffDistance=3250
    hPFFrequencyAtMax=300
    lPFFrequencyAtMax=800
    lPFRadiusMax=3000
    lPFRadiusMin=1000
    occlusionInterpolationTime=0.5
    occlusionLowPassFilterFrequency=6000
    occlusionTraceChannel=ECC_GameTraceChannel11
    occlusionVolumeAttenuation=3
    spatializationAlgorithm=SPATIALIZATION_HRTF
    stereoSpread=0
    attenuationShapeExtents {x=4000}
    customAttenuationCurve {editorCurveData {keys [
       <at 0 1 RCIM_Cubic>
       {arriveTangent=-1.7422, interpMode=RCIM_Cubic, leaveTangent=-1.7422, time=0.1205, value=0.8944}
       {arriveTangent=-1.017, interpMode=RCIM_Cubic, leaveTangent=-1.017, time=0.3914, value=0.3181}
       {arriveTangent=-0.5016, interpMode=RCIM_Cubic, leaveTangent=-0.5016, tangentMode=RCTM_User, time=1, value=0}
      ]}}
   }
  }
 }
 {
  id=Enveloper_0
  children [WavePlayer_4]
  properties {pitchCurve {editorCurveData {keys [{time=0}]}}, volumeCurve {editorCurveData {keys [<at 0 0 RCIM_Cubic>,<at 0.5 1>]}}}
 }
 {
  id=Enveloper_1
  children [WavePlayer_1]
  properties {pitchCurve {editorCurveData {keys [{time=0}]}}, volumeCurve {editorCurveData {keys [<at 0 0 RCIM_Cubic>,<at 1 1>]}}}
 }
 {
  id=Enveloper_2
  children [WavePlayer_11]
  properties {pitchCurve {editorCurveData {keys [{time=0}, <at 4 1.5>]}}, volumeCurve {editorCurveData {keys [<at 0 0>,<at 0.5 1>]}}}
 }
 {
  id=Enveloper_3
  children [WavePlayer_13]
  properties {pitchCurve {editorCurveData {keys [{time=0}]}}, volumeCurve {editorCurveData {keys [<at 0 0 RCIM_Cubic>,<at 1 1>]}}}
 }
 {
  id=Enveloper_4
  children [WavePlayer_12]
  properties {pitchCurve {editorCurveData {keys [{time=0}]}}, volumeCurve {editorCurveData {keys [<at 0 0 RCIM_Cubic>,<at 1 1>]}}}
 }
 {
  id=Enveloper_5
  children [WavePlayer_0]
  properties {pitchCurve {editorCurveData {keys [{time=0}]}}, volumeCurve {editorCurveData {keys [<at 0 0 RCIM_Cubic>,<at 1 1>]}}}
 }
 {
  id=Enveloper_6
  children [WavePlayer_2]
  properties {pitchCurve {editorCurveData {keys [{time=0}]}}, volumeCurve {editorCurveData {keys [<at 0 0 RCIM_Cubic>,<at 1 1>]}}}
 }
 {id=Mixer_0, children [ModulatorContinuous_4, ModulatorContinuous_0, ModulatorContinuous_1], properties {inputVolume [1,1,1]}}
 {id=Mixer_1, children [ModulatorContinuous_6], properties {inputVolume [1.75]}}
 {id=Mixer_3, children [ModulatorContinuous_5], properties {inputVolume [0.8]}}
 {id=Mixer_5, children [ModulatorContinuous_3, ModulatorContinuous_2], properties {inputVolume [1,1.5]}}
 {
  id=ModulatorContinuous_0
  children [Enveloper_0]
  properties {
   pitchModulationParams {default=0, maxInput=16384, maxOutput=1.2, minInput=512, minOutput=0.8, paramMode=MPM_Abs, parameterName=FallDistance}
   volumeModulationParams {default=0, maxInput=4096, minInput=512, paramMode=MPM_Abs, parameterName=FallDistance}
  }
 }
 {
  id=ModulatorContinuous_1
  children [Enveloper_1]
  properties {
   pitchModulationParams {default=0, maxInput=512, maxOutput=1.2, minOutput=0.8, paramMode=MPM_Abs, parameterName=FallDistance}
   volumeModulationParams {default=0, maxInput=512, maxOutput=1.5, paramMode=MPM_Abs, parameterName=FallDistance}
  }
 }
 {
  id=ModulatorContinuous_2
  children [Enveloper_4]
  properties {
   pitchModulationParams {default=0, maxInput=4096, maxOutput=1.2, minInput=512, minOutput=0.8, paramMode=MPM_Abs, parameterName=FallDistance}
   volumeModulationParams {default=0, maxInput=4096, minInput=512, paramMode=MPM_Abs, parameterName=FallDistance}
  }
 }
 {
  id=ModulatorContinuous_3
  children [Enveloper_3]
  properties {
   pitchModulationParams {default=0, maxInput=512, maxOutput=1.2, minOutput=0.8, paramMode=MPM_Abs, parameterName=FallDistance}
   volumeModulationParams {default=0, maxInput=512, maxOutput=1.5, paramMode=MPM_Abs, parameterName=FallDistance}
  }
 }
 {
  id=ModulatorContinuous_4
  children [Enveloper_6]
  properties {
   pitchModulationParams {default=0, maxInput=1300, minInput=512, minOutput=1, paramMode=MPM_Abs, parameterName=FallDistance}
   volumeModulationParams {default=0, maxInput=1300, minInput=512, paramMode=MPM_Abs, parameterName=FallDistance}
  }
 }
 {
  id=ModulatorContinuous_5
  children [Enveloper_2]
  properties {
   pitchModulationParams {default=0, maxInput=2048, maxOutput=1.2, minOutput=0.8, paramMode=MPM_Abs, parameterName=FallDistance}
   volumeModulationParams {default=0, maxInput=2048, paramMode=MPM_Abs, parameterName=FallDistance}
  }
 }
 {
  id=ModulatorContinuous_6
  children [Enveloper_5]
  properties {
   pitchModulationParams {default=0, maxInput=2048, maxOutput=1.2, minOutput=0.8, paramMode=MPM_Abs, parameterName=FallDistance}
   volumeModulationParams {default=0, maxInput=2048, paramMode=MPM_Abs, parameterName=FallDistance}
  }
 }
 {id=QualityLevel_0, children [Mixer_0,Mixer_0,Mixer_3,Mixer_3]}
 {id=QualityLevel_2, children [Mixer_5,Mixer_5,Mixer_1,Mixer_1]}
 {id=SoundClass_0, children [Attenuation_0], properties {soundClassOverride=&Fort_Player_Foley_Athena}}
 {
  id=SoundClass_1
  children [Attenuation_1]
  properties {soundClassOverride=&Fort_Player_Foley_3P_Enemy_Athena}
 }
 {id=Switch_0, children [SoundClass_0,SoundClass_0,SoundClass_1], properties {intParameterName=Affiliation}}
 {id=WavePlayer_0, properties {bLooping=true, soundWaveAssetPtr=&/Character_InAir_Wind_Lp_Enemy_01}}
 {id=WavePlayer_1, properties {bLooping=true, soundWaveAssetPtr=/Game/Athena/Sounds/Wind_Falling/Character_InAir_Wind_Lp_02}}
 {id=WavePlayer_11, properties {bLooping=true, soundWaveAssetPtr=/Game/Athena/Sounds/Wind_Falling/Wind_UmbrellaClosed_Loop_1P_opt_02}}
 {id=WavePlayer_12, properties {bLooping=true, soundWaveAssetPtr=&Character_InAir_Wind_Lp_Enemy_01}}
 {id=WavePlayer_13, properties {bLooping=true, soundWaveAssetPtr=/Game/Athena/Sounds/Wind_Falling/Character_InAir_AudioTell_01}}
 {id=WavePlayer_2, properties {bLooping=true, soundWaveAssetPtr=/Game/Athena/Sounds/Wind_Falling/AMB_Wind_Skydive_Cloth_Flap_Loop_01}}
 {id=WavePlayer_4, properties {bLooping=true, soundWaveAssetPtr=/Game/Athena/Sounds/Wind_Falling/AMB_Wind_Skydive_Vertical_Loop_01}}
]
properties {
 bOverrideConcurrency=true
 maxDistance=7250
 volumeMultiplier=2
 assetUserData [&self:AudioAssetUserData_0]
 concurrencyOverrides {maxCount=3, resolutionRule=StopQuietest}
}
